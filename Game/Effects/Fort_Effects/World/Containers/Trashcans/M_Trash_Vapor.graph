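# Translucent, unlit, two-sided mesh-particle material for vapor, filed under trashcan container effects.
# Panning goo and smoke textures, offset in V by a dynamic parameter, multiply together over a dark olive base, then tint by particle color through an emissive-adjustment function.
# Particle alpha, smoke alpha and a vertical fade ramp form a soft mask.
# SplineThicken's world position offset is combined with a sine-modulated scrolling ramp for animated displacement.
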Material /Game/Effects/Fort_Effects/World/Containers/Trashcans/M_Trash_Vapor
graph {
 properties {
  bUsedWithMeshParticles=true
  blendMode=BLEND_Translucent
  pixelDepthOffsetMode=PDOM_Legacy
  shadingModel=MSM_Unlit
  twoSided=true
  usedShadingModels=Unlit
 }
 nodes [
  n1 = TextureCoordinate()
  Add(A=@n1, B=AppendVector(A=Constant() {r=1}, B=Constant() {r=0.5}))
  n2 = Panner(Coordinate=TextureCoordinate() {vTiling=0.5}) {speedY=0.1}
  n3 = Time()
  Add(
   A=Multiply(A=TextureSample(UVs=@n2) {texture=/Game/Global/Textures/Default/Masks_and_Gradients/T_Water_Top_Ramp}.G, B=Sine(None=@n3) {period=5})
   B=MaterialFunctionCall(WidthBase=Constant() {r=60}, WidthTip=Constant() {r=60}, Expand U or V UV Channel=StaticBool() {value=true}) {materialFunction=&SplineThicken}.WorldPositionOffset
  )
  Constant() {r=1}
  n4 = Constant3Vector() {constant=<rgba 0.05 0.05 0.01 0>}
  n5 = Add(A=TextureCoordinate() {vTiling=2}, B=AppendVector(A=Constant(), B=DynamicParameter() {defaultValue=<rgba 0 0 0 1>}.Param1))
  n6 = TextureSample(UVs=Panner(Coordinate=@n5) {speedY=0.3}) {texture=/Game/Effects/Fort_Effects/Textures/Smoke/T_Limbo_Smoke}
  n7 = Multiply(
   A=TextureSample(UVs=Panner(Coordinate=@n5) {speedX=0.1, speedY=0.4}) {texture=/Game/Effects/Fort_Effects/Textures/T_Swril_Goo_Tile}.RGB
   B=@n6.RGB
  )
  n8 = ParticleColor()
  MaterialFunctionCall(Emissive=Multiply(A=Add(A=@n4, B=@n7), B=@n8.RGB)) {materialFunction=&MF_FX_EmissiveAdjustments}
  Multiply(A=@n8.A, B=Multiply(A=@n6.A, B=TextureSample() {texture=/Game/Effects/Fort_Effects/Textures/Ramps/T_Ramp_Vert_Soft_Fade}.G))
  n9 = Subtract(A=Panner(Coordinate=ComponentMask(None=@n1) {g=true}) {speedY=0.2}, B=Constant() {r=0.5})
  Power(Base=OneMinus(None=Multiply(A=@n9, B=@n9)), Exp=Constant() {r=20})
 ]
}
